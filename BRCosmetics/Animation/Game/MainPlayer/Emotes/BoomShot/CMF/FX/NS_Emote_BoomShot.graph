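# Niagara effect for the BoomShot emote: bombs rise and spread into a heart, then give way to a heart-shaped explosion with comic text.
# Bomb particles lerp from a small sphere to a curve-shaped ring with a height curve, emitting death events as they expire.
# Timed bursts add glow, small hearts, two staggered KA-BOOM text meshes, and smoke sampled from a heart static mesh.
# Impact flashes, Ben-Day dot flashes, sparks and outline smoke complete the blast.
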NiagaraSystem /BRCosmetics/Animation/Game/MainPlayer/Emotes/BoomShot/CMF/FX/NS_Emote_BoomShot
dependencies {
 usedDataInterfaces [
  &NiagaraDataInterfaceCurve
  &NiagaraDataInterfaceColorCurve
  &NiagaraDataInterfaceMeshRendererInfo
  &NiagaraDataInterfaceVector2DCurve
  &NiagaraDataInterfaceStaticMesh
  &NiagaraDataInterfaceVectorCurve
 ]
 usedDynamicInputs [
  &Vector4FromCurve
  &SimulationPosition
  &Multiply_Vector2
  &MakeVector2D
  &UniformAOrBFloat
 ]
 usedModules [
  &SystemState
  &self:SystemSpawnScript.SystemScriptSource.SystemScriptGraph.NiagaraNodeAssignment_3.SetVariables_#1
 ]
}
emitters {
 Bombs_HeartShape {
  emitterUpdateScript [
   &EmitterState()
   &SpawnBurst_Instantaneous(Spawn Count=15, Use Spawn Probability=false, UseLoopCountLimit=false)
  ]
  particleSpawnScript [
   &InitializeParticle(
    Color Maximum=<rgba 1 0.1134 0 1>
    Color Minimum=<rgba 1 0.33 0 1>
    Color Mode=::NewEnumerator2
    Lifetime Max=0.9
    Lifetime Min=0.8
    Lifetime Mode=::NewEnumerator1
    Mesh Scale Mode=::NewEnumerator3
    Sprite Size Mode=::NewEnumerator3
    Uniform Sprite Size=13
    UsePositionOffset=false
    Write Material Random=true
   )
   &ShapeLocation(MaskSpawnGroup=false, Offset=<0 0 20>, Offset Mode=::NewEnumerator0, Override Seed=false, Sphere Radius=10, UseOwnerScale=true)
   &ShapeLocation(
    MaskSpawnGroup=false
    Override Seed=false
    Radius Position:NiagaraFloat=&FloatFromCurve(
     FloatCurve=DataInterface() {propertyValues {Curve {keys [
         {arriveTangent=-1.952, interpMode=RCIM_Cubic, leaveTangent=-1.952, tangentMode=RCTM_User, time=0, value=0.75}
         {arriveTangent=-0.417, arriveTangentWeight=0.2482, interpMode=RCIM_Cubic, leaveTangent=-0.417, leaveTangentWeight=0.1245, time=0.666, value=1}
         {arriveTangent=-3.1386, interpMode=RCIM_Cubic, leaveTangent=-3.1386, tangentMode=RCTM_User, time=1, value=0.333}
        ]}}}
     CurveIndex=&Abs_Float(Float:NiagaraFloat=/Niagara/DynamicInputs/Multiply/ScaleAndBiasFloat(
       Float:NiagaraFloat=&ReturnNormalizedExecIndex(Normalized Index Includes One=false)
       Scale:NiagaraFloat=2
       Bias:NiagaraFloat=-1
      ))
    )
    Ring / Disc Distribution Mode:ENiagaraSphereDistributionMode=::NewEnumerator1
    Rotation Angle=0.25
    Rotation Axis=<0 -1 0>
    Rotation Mode=::NewEnumerator1
    Shape Primitive=::NewEnumerator4
    U Position=&FloatFromCurve(
     FloatCurve=DataInterface() {propertyValues {Curve {keys [
         <at 0 0>
         <at 0.1 0.0666>
         <at 0.3 0.25>
         {arriveTangent=0.8541, time=0.5, value=0.5}
         <at 0.7 0.75>
         <at 0.9 0.9333>
         {arriveTangent=1.5135, time=1, value=1}
        ]}}}
     CurveIndex=&ReturnNormalizedExecIndex(Normalized Index Includes One=false)
    )
    UseOwnerScale=true
   ) {module=ShapeLocation001}
   &ApplyInitialForces()
   &InitialMeshOrientation(Use Rotation Vector=false)
  ]
  particleUpdateScript [
   &ParticleState()
   &UpdateMeshOrientation(Rotation Rate=&FloatFromCurve(FloatCurve=DataInterface() {propertyValues {Curve {keys [
         {interpMode=RCIM_Cubic, tangentMode=RCTM_User, time=0, value=0}
         <at 0.5 0.5 RCIM_Cubic>
         {interpMode=RCIM_Cubic, tangentMode=RCTM_User, time=1, value=0}
        ]}}})) {module=UpdateMeshOrientation001}
   /SetVariables() {isSetParameters=true}
   /SetVariables() {isSetParameters=true}
   /SetVariables(
    Particles.FlareCamera=5
    Particles.FlareColor=&MultiplyLinearColor(
     Linear Color=&RandomRangeLinearColor(Minimum=<rgba 2 0.697 0.0553 1>, Maximum=<rgba 2 0.2912 0.0727 1>)
     Scale Factor=&MakeVector4(
      X=1
      Y=1
      Z=1
      W=&Waveform(
       "Waveform [1]"=::NewEnumerator0
       "[1] Phase Scale"=&RandomRangeFloat(Minimum=2, Maximum=4)
       "[1] Frequency"=1
       "[1] Amplitude Min/Max"=&RandomRangeVector2D(Minimum=<0.1 2>, Maximum=<0.5 4>)
      )
     )
    )
    Particles.FlarePosition:NiagaraPosition=&AddVectorToPosition(Position=$Particles.Position, Vector=<0 0 0>)
   ) {isSetParameters=true}
   /SetVariables(Particles.Position=&Lerp_Vector(
     A=$Particles.Initial.ShapeLocation.ShapePosition
     B=$Particles.Initial.ShapeLocation001.ShapePosition
     Alpha=&FloatFromCurve(FloatCurve=DataInterface() {propertyValues {Curve {keys [
          {arriveTangent=-0.0702, interpMode=RCIM_Cubic, leaveTangent=-0.0702, tangentMode=RCTM_User, time=0, value=0}
          {arriveTangent=1.0501, interpMode=RCIM_Cubic, tangentMode=RCTM_Break, time=0.6, value=1}
         ]}}})
    )) {isSetParameters=true}
   /SetVariables(Particles.Position=&AddVectorToPosition(
     Position=$Particles.Position
     Vector=&MakeVector(
      X=0
      Y=0
      Z=&FloatFromCurve(
       FloatCurve=DataInterface() {propertyValues {Curve {keys [
           {arriveTangent=5.1878, interpMode=RCIM_Cubic, leaveTangent=5.1878, tangentMode=RCTM_User, time=0, value=0}
           {interpMode=RCIM_Cubic, leaveTangent=0.3845, tangentMode=RCTM_Break, time=0.5, value=1}
           {interpMode=RCIM_Cubic, leaveTangent=-0.6667, time=1, value=1.1}
          ]}}}
       Scale Curve=110
      )
     )
    )) {isSetParameters=true}
   &SolveForcesAndVelocity(Clamp Velocity=false, Limit Acceleration=false)
   &ScaleColor(
    Scale Alpha=&FloatFromCurve(FloatCurve=DataInterface() {propertyValues {Curve {keys [
         {interpMode=RCIM_Cubic, leaveTangent=-11.4763, tangentMode=RCTM_Break, time=0.8, value=1}
         {interpMode=RCIM_Cubic, tangentMode=RCTM_User, time=1, value=0}
        ]}}})
    Scale RGB=&VectorFromFloat(Value=&FloatFromCurve(
      FloatCurve=DataInterface() {propertyValues {Curve {keys [
          {interpMode=RCIM_Cubic, tangentMode=RCTM_User, time=0, value=0}
          <at 0.1 0.95 RCIM_Cubic>
          {interpMode=RCIM_Cubic, tangentMode=RCTM_User, time=1, value=0}
         ]}}}
      Scale Curve=&RandomRangeFloat(Minimum=2, Maximum=5)
     ))
    ScaleA=false
    ScaleRGB=true
   )
   &GenerateDeathEvent(Use Event Probability=false)
  ]
  rendererClasses [&NiagaraMeshRendererProperties, &NiagaraSpriteRendererProperties]
  renderers [
   {rendererClass=&NiagaraMeshRendererProperties, rendererIndex=0}
   {rendererClass=&NiagaraSpriteRendererProperties, rendererIndex=1}
  ]
 }
 Flash_BenDayDots {
  emitterUpdateScript [&EmitterState()]
  particleSpawnScript [
   &InitializeParticle(
    Color=<rgba 1 0.2686 0.0219 0.5>
    Lifetime=0.2
    Sprite Size Mode=::NewEnumerator3
    Uniform Sprite Size=&Multiply_Float(A=400, B=$System.TempScale)
    UsePositionOffset=false
    Write Material Random=true
   )
   &DynamicMaterialParameters(
    Index 0 Param 1=0.2
    Index 0 Param 2=1
    Module.Param0WriteEnabled=true
    Module.Param1WriteEnabled=true
    Module.Param2WriteEnabled=true
    Module.Param3WriteEnabled=true
   )
   &SystemLocation(MaskSpawnGroup=false)
  ]
  particleUpdateScript [
   &ParticleState()
   &ScaleSpriteSize(Uniform Curve Sprite Scale=DataInterface() {propertyValues {Curve {keys [
        {arriveTangent=4.1785, interpMode=RCIM_Cubic, leaveTangent=4.1785, tangentMode=RCTM_User, time=0, value=0}
        {arriveTangent=1.1179, interpMode=RCIM_Cubic, leaveTangent=1.1179, tangentMode=RCTM_User, time=0.2042, value=0.6228}
        {arriveTangent=0.0224, interpMode=RCIM_Cubic, leaveTangent=0.0224, tangentMode=RCTM_User, time=1, value=1}
       ]}}})
   /SetVariables(
    Particles.BenDayDotCamera=25
    Particles.BenDayDotColor:LinearColor=&MultiplyLinearColor(
     Linear Color=&ColorFromCurve(ColorCurve=DataInterface() {propertyValues {
        AlphaCurve {keys [<at 0.4073 1 RCIM_Cubic>]}
        BlueCurve {keys [<at 0.2311 0.045 RCIM_Cubic>, {interpMode=RCIM_Cubic, leaveTangent=0.0649, time=0.41, value=0.1937}, <at 0.6851 0.0833>]}
        GreenCurve {keys [<at 0.2311 0.6513 RCIM_Cubic>, {interpMode=RCIM_Cubic, leaveTangent=-1.1021, time=0.41, value=0.075}, <at 0.6851 0>]}
        RedCurve {keys [<at 0.2311 1.5 RCIM_Cubic>, {interpMode=RCIM_Cubic, leaveTangent=-0.8461, time=0.41, value=1.5}, <at 0.6851 1>]}
       }})
     Scale Factor=<1 1 1 1>
    )
    Particles.BenDayDotDMP:Vector4f=&MakeVector4(
     X=&FloatFromCurve(FloatCurve=DataInterface() {propertyValues {Curve {keys [{interpMode=RCIM_Cubic, tangentMode=RCTM_User, time=0, value=0}, {interpMode=RCIM_Cubic, tangentMode=RCTM_User, time=1, value=1}]}}})
     Y=&FloatFromCurve(
      FloatCurve=DataInterface() {propertyValues {Curve {keys [
          {interpMode=RCIM_Cubic, tangentMode=RCTM_User, time=0, value=1}
          {interpMode=RCIM_Cubic, tangentMode=RCTM_User, time=0.5, value=0}
          {tangentMode=RCTM_User, time=1, value=1}
         ]}}}
      Scale Curve=0.4
     )
     Z=0
     W=0
    )
    Particles.BenDayDotScale:Vector2f=&Vector2DFromFloat(Value=&Multiply_Float(A=&RandomRangeFloat(Minimum=120, Maximum=200), B=$System.TempScale))
   ) {isSetParameters=true}
   &ScaleColor(
    Scale Alpha=&FloatFromCurve(FloatCurve=DataInterface() {propertyValues {Curve {keys [
         {interpMode=RCIM_Cubic, tangentMode=RCTM_User, time=0, value=1}
         {arriveTangent=-2.5164, interpMode=RCIM_Cubic, leaveTangent=-2.5164, tangentMode=RCTM_User, time=1, value=0}
        ]}}})
    Scale RGB=&VectorFromFloat(Value=&FN_FrontEndFloatModifier(Gameplay=1.5, Front End=1))
    ScaleA=true
    ScaleRGB=true
   )
  ]
  rendererClasses [&NiagaraSpriteRendererProperties]
  renderers [
   {rendererClass=&NiagaraSpriteRendererProperties}
   {rendererClass=&NiagaraSpriteRendererProperties, rendererIndex=1}
  ]
 }
 Glow_BenDayDots {
  emitterUpdateScript [
   &EmitterState()
   &SpawnBurst_Instantaneous(Spawn Time=0.85, Use Spawn Probability=false, UseLoopCountLimit=false)
  ]
  particleSpawnScript [
   &InitializeParticle(
    Lifetime=0.3
    Sprite Rotation Mode=::NewEnumerator1
    Sprite Size Mode=::NewEnumerator3
    Uniform Sprite Size=150
    UsePositionOffset=false
    Write Material Random=true
   )
   /SetVariables(Particles.GlowLifetime=0.2) {isSetParameters=true}
   &InitialMeshOrientation(
    Rotation=&RandomRangeVector(Minimum=<-180 -180 -180>, Maximum=<180 180 180>)
    Use Orientation Vector=true
    Use Rotation Vector=true
   )
   /SetVariables(Particles.Position=&AddVectorToPosition(Position=$Particles.Position, Vector=<0 0 150>)) {isSetParameters=true}
  ]
  particleUpdateScript [
   &ParticleState()
   &ScaleSpriteSize(Uniform Curve Sprite Scale=DataInterface() {propertyValues {Curve {keys [
        {interpMode=RCIM_Cubic, tangentMode=RCTM_User, time=0, value=1}
        <at 0.2 3 RCIM_Cubic>
        {interpMode=RCIM_Cubic, tangentMode=RCTM_User, time=1, value=1}
       ]}}}) {module=ScaleSpriteSize001}
   /SetVariables(Particles.GlowUpdateAge:NiagaraFloat=&Divide_Float(A=$Particles.Age, B=$Particles.Initial.GlowLifetime)) {isSetParameters=true}
   /SetVariables(
    Particles.GlowCamera=-44
    Particles.GlowColor=&MakeLinearColorFromVectorAndFloat(
     "Vector (RGB)"=&Multiply_Vector(
      A=&Multiply_Vector(
       A=&MakeVectorFromLinearColorRGB(Color=<rgba 1 0.008 0.9056 1>)
       B=&VectorFromCurve(
        VectorCurve=DataInterface() {propertyValues {
          XCurve {keys [<at 0 1>,<at 0.25 0.5>]}
          YCurve {keys [<at 0 1>,<at 0.25 0.5>,<at 0.5 0.5>,<at 1 0.5>]}
          ZCurve {keys [<at 0 1>,<at 0.25 0.5>]}
         }}
        CurveIndex=$Particles.GlowUpdateAge
       )
      )
      B=&VectorFromFloat(Value=&FloatFromCurve(
        FloatCurve=DataInterface() {propertyValues {Curve {keys [
            <at 0 0 RCIM_Cubic>
            {arriveTangent=0.3223, interpMode=RCIM_Cubic, leaveTangent=0.3223, tangentMode=RCTM_User, time=0.1, value=1}
            {arriveTangent=0.0367, interpMode=RCIM_Cubic, leaveTangent=-3.7949, tangentMode=RCTM_Break, time=0.3, value=1}
            <at 1.0011 0.002 RCIM_Cubic>
           ]}}}
        CurveIndex=$Particles.GlowUpdateAge
        Scale Curve=2
       ))
     )
     "Float (Alpha)"=1
    )
    Particles.GlowDMP:Vector4f=&MakeVector4(
     X=&FloatFromCurve(
      FloatCurve=DataInterface() {propertyValues {Curve {keys [{arriveTangent=0.0592, interpMode=RCIM_Cubic, time=0, value=0.1}, <at 1 1 RCIM_Cubic>]}}}
      CurveIndex=$Particles.GlowUpdateAge
     )
     Y=50
     Z=0
     W=0
    )
    Particles.GlowScale:Vector3f=&VectorFromFloat(Value=3)
   ) {isSetParameters=true}
   &DynamicMaterialParameters(
    Index 0 Param 1=&FloatFromCurve(
     FloatCurve=DataInterface() {propertyValues {Curve {keys [
         {interpMode=RCIM_Cubic, tangentMode=RCTM_User, time=0, value=0}
         {arriveTangent=2.6011, interpMode=RCIM_Cubic, leaveTangent=2.6011, tangentMode=RCTM_User, time=1, value=1}
        ]}}}
     Scale Curve=0.3
    )
    Index 0 Param 2=&FloatFromCurve(
     FloatCurve=DataInterface() {propertyValues {Curve {keys [
         {interpMode=RCIM_Cubic, tangentMode=RCTM_User, time=0, value=0.4}
         {interpMode=RCIM_Cubic, tangentMode=RCTM_User, time=0.5, value=1}
         {tangentMode=RCTM_User, time=1, value=0.4}
        ]}}}
     Scale Curve=0.25
    )
    Module.Index0Param0Write=true
    Module.Index0Param1Write=true
    Module.Index0Param2Write=true
    Module.Index0Param3Write=true
   ) {module=DynamicMaterialParameters002}
   &Color(Color=&ColorFromCurve(ColorCurve=DataInterface() {propertyValues {
       AlphaCurve {keys [<at 0.4073 1 RCIM_Cubic>]}
       BlueCurve {keys [<at 0 8.0361>,<at 0.6085 0.9933>,<at 0.891 0.5455>]}
       GreenCurve {keys [<at 0 4.2012>,<at 0.6085 0>,<at 0.891 0.3522>]}
       RedCurve {keys [<at 0 9>,<at 0.6085 1>,<at 0.891 0.5941>]}
      }})) {module=Color001}
   &ScaleColor(
    Color Value To Scale=$Particles.Color
    Scale Alpha=&FloatFromCurve(
     FloatCurve=DataInterface() {propertyValues {Curve {keys [
         {interpMode=RCIM_Cubic, tangentMode=RCTM_User, time=0.3, value=1}
         {arriveTangent=-2.5266, interpMode=RCIM_Cubic, leaveTangent=-2.5266, tangentMode=RCTM_User, time=1, value=0}
        ]}}}
     Scale Curve=&FNTimeOfDayScalar(Night Scale=1, Morning Scale=1, Evening Scale=1)
    )
    Scale RGB=&VectorFromFloat(Value=&FNTimeOfDayScalar(Night Scale=1, Morning Scale=1, Evening Scale=1))
    ScaleA=true
    ScaleRGB=true
   )
   &CameraOffset(Camera Offset Amount=-44) {module=CameraOffset001}
  ]
  rendererClasses [&NiagaraSpriteRendererProperties, &NiagaraMeshRendererProperties]
  renderers [
   {rendererClass=&NiagaraSpriteRendererProperties}
   {rendererClass=&NiagaraMeshRendererProperties, rendererIndex=1}
  ]
 }
 Hearts_Small {
  emitterUpdateScript [
   &EmitterState()
   &SpawnBurst_Instantaneous(Spawn Count=5, Spawn Time=0.8, Use Spawn Probability=false, UseLoopCountLimit=false)
  ]
  particleSpawnScript [
   &InitializeParticle(
    Color Maximum=<rgba 0.9928 0.0647 0.353 1>
    Color Minimum=<rgba 1 0 0.0615 1>
    Color Mode=::NewEnumerator2
    Lifetime Max=2.9
    Lifetime Min=1.5
    Lifetime Mode=::NewEnumerator1
    Sprite Size Max=<25 35>
    Sprite Size Min=<15 15>
    Sprite Size Mode=::NewEnumerator2
    UsePositionOffset=false
    Write Material Random=true
   ) {module=InitializeParticle001}
   &ShapeLocation(MaskSpawnGroup=false, Override Seed=false, Sphere Radius=9, UseOwnerScale=true) {module=ShapeLocation001}
   &AddVelocity(
    Module.Use Falloff Distance=false
    Override Seed=false
    Velocity Mode=::NewEnumerator1
    Velocity Origin=$System.DeathPosition
    Velocity Speed=&RandomRangeFloat(Minimum=150, Maximum=200)
   ) {module=AddVelocity001}
   &ApplyInitialForces()
   /SetVariables(Particles.Position=&AddVectorToPosition(Position=$Particles.Position, Vector=<0 0 100>)) {isSetParameters=true}
  ]
  particleUpdateScript [
   &ParticleState()
   &SpriteRotationRate(Rotation Rate=&RandomRangeFloat(Minimum=-90, Maximum=90))
   &ScaleSpriteSize(Uniform Curve Sprite Scale=DataInterface() {propertyValues {Curve {keys [
        {interpMode=RCIM_Cubic, tangentMode=RCTM_User, time=0, value=0}
        {arriveTangent=3.3333, interpMode=RCIM_Cubic, leaveTangent=3.3333, time=0.2, value=2}
        {interpMode=RCIM_Cubic, tangentMode=RCTM_User, time=0.3, value=1}
        {tangentMode=RCTM_User, time=1, value=0}
       ]}}}) {module=ScaleSpriteSize001}
   &AccelerationForce(Acceleration=&RandomRangeVector(Minimum=<0 70 200>, Maximum=<0 222 333>))
   &Drag(
    Drag=&FloatFromCurve(
     FloatCurve=DataInterface() {propertyValues {Curve {keys [
         {arriveTangent=2.492, interpMode=RCIM_Cubic, leaveTangent=2.492, tangentMode=RCTM_User, time=0, value=0}
         {interpMode=RCIM_Cubic, tangentMode=RCTM_User, time=1, value=1}
        ]}}}
     Scale Curve=9
    )
    Use Linear Drag=true
    Use Rotational Drag=true
   )
   &CurlNoiseForce(Bias Noise=false, Noise Frequency=40.3446, Noise Strength=11.6461, Pan Noise=true, Pan Noise Field=<11 0 0>)
   &SolveForcesAndVelocity(Clamp Velocity=false, Limit Acceleration=false)
   &ScaleColor(
    Scale Alpha=&FloatFromCurve(
     FloatCurve=DataInterface() {propertyValues {Curve {keys [
         <at 0 0>
         <at 0.0805 1.8557>
         {arriveTangent=-2.0136, interpMode=RCIM_Cubic, leaveTangent=-12.6391, tangentMode=RCTM_Break, time=0.608, value=1.86}
         <at 1.0011 0.002>
        ]}}}
     Scale Curve=&RandomRangeFloat(Minimum=0.4, Maximum=0.7)
    )
    Scale RGB=&MakeVector(
     X=&FloatFromCurve(FloatCurve=DataInterface() {propertyValues {Curve {keys [
          <at 0 22 RCIM_Cubic>
          {arriveTangent=-21, interpMode=RCIM_Cubic, leaveTangent=-0.2565, tangentMode=RCTM_Break, time=0.2, value=1}
          <at 1 1 RCIM_Cubic>
         ]}}})
     Y=1
     Z=1
    )
    ScaleA=true
    ScaleRGB=true
   ) {module=ScaleColor001}
   &CameraOffset(Camera Offset Amount=22)
  ]
  rendererClasses [&NiagaraSpriteRendererProperties]
  renderers [{rendererClass=&NiagaraSpriteRendererProperties}]
 }
 ImpactFlash {
  emitterUpdateScript [&EmitterState(
    Enable Distance Culling=true
    Max Distance Response=::KillImmediately
    Module.bUseMaxDistance=true
    Module.bUseMinDistance=false
    Scalability Mode=::NewEnumerator1
   )]
  particleSpawnScript [
   &InitializeParticle(
    Color=<rgba 2 0.3333 0 1>
    Lifetime=0.15
    Sprite Rotation Mode=::NewEnumerator2
    Sprite Size Mode=::NewEnumerator3
    Uniform Sprite Size=&Multiply_Float(A=$System.TempScale, B=&UniformRangedFloat(Minimum=168, Maximum=300))
    UsePositionOffset=false
    Write Material Random=true
   )
   &CameraOffset(Camera Offset Amount=15)
   /SetVariables(Particles.RandomFloat=&UniformRangedFloat(Minimum=0, Maximum=1)) {isSetParameters=true}
   /SetVariables(
    Particles.SubImageIndex=&RandomRangeFloat(Minimum=0, Maximum=3)
    Particles.SubImageIndexAlt:NiagaraFloat=&RandomRangeFloat(Minimum=0, Maximum=3)
   ) {isSetParameters=true}
  ]
  particleUpdateScript [
   &ParticleState()
   &ScaleColor(
    Scale Alpha=&Multiply_Float(
     A=&FloatFromCurve(FloatCurve=DataInterface() {propertyValues {Curve {keys [{interpMode=RCIM_Cubic, tangentMode=RCTM_User, time=0.5, value=1}, {interpMode=RCIM_Cubic, tangentMode=RCTM_User, time=1, value=0}]}}})
     B=&FloatFromCurve(
      FloatCurve=DataInterface() {propertyValues {Curve {keys [
          {interpMode=RCIM_Cubic, tangentMode=RCTM_User, time=0, value=0.3}
          {interpMode=RCIM_Cubic, tangentMode=RCTM_User, time=0.5, value=1}
          {tangentMode=RCTM_User, time=1, value=0.3}
         ]}}}
      CurveIndex=Linked() {linkedVariable {name=NPC.FortniteNPC.TimeOfDay_0-1, type=NiagaraFloat}}
     )
    )
    ScaleA=true
    ScaleRGB=false
   ) {module=ScaleColor001}
   &ScaleSpriteSize(Uniform Curve Sprite Scale=DataInterface() {propertyValues {Curve {keys [
        {interpMode=RCIM_Cubic, tangentMode=RCTM_User, time=0, value=0}
        <at 0.2 1 RCIM_Cubic>
        {interpMode=RCIM_Cubic, tangentMode=RCTM_User, time=1, value=0}
       ]}}}) {module=ScaleSpriteSize002}
   &GenerateDeathEvent(Use Event Probability=false)
  ]
  rendererClasses [&NiagaraSpriteRendererProperties]
  renderers [
   {rendererClass=&NiagaraSpriteRendererProperties}
   {rendererClass=&NiagaraSpriteRendererProperties, rendererIndex=1}
  ]
 }
 Smoke_HeartShape {
  emitterUpdateScript [
   &EmitterState()
   &SpawnBurst_Instantaneous(Spawn Count=33, Spawn Time=0.97, Use Spawn Probability=false, UseLoopCountLimit=false)
  ]
  particleSpawnScript [
   &InitializeParticle(
    Color Mode=::NewEnumerator0
    Lifetime Max=2.7
    Lifetime Min=0.7
    Lifetime Mode=::NewEnumerator1
    Mesh Scale Mode=::NewEnumerator1
    Mesh Uniform Scale Max=0.65
    Mesh Uniform Scale Min=0.75
    Sprite Size Mode=::NewEnumerator3
    Uniform Sprite Size=15
    UsePositionOffset=false
    Write Material Random=true
   )
   /Niagara/Modules/Spawn/Location/StaticMeshLocation(
    MaskSpawnGroup=false
    Mesh Sampling Type:ENiagara_StaticSamplingMode=::NewEnumerator2
    OffsetAlongNormal=false
    OverrideRandomSeed=false
    SamplePosOffset=false
    Static Mesh=DataInterface() {propertyValues {
      DefaultMesh=&/SM_Emote_BoomShot_Heart
      InstanceIndex=-1
      LODIndex=0
      PreviewMesh=&SM_Emote_BoomShot_Heart
      SoftSourceActor=null
      SourceMode=DefaultMeshOnly
      bAllowSamplingFromStreamingLODs=false
      bCaptureTransformsPerFrame=true
      bUsePhysicsBodyVelocity=false
      FilteredSockets []
      LODIndexUserParameter {parameter {name=None, typeDefHandle {registeredTypeIndex=108}, varData []}}
      MeshParameterBinding {parameter {name=None, typeDefHandle {registeredTypeIndex=1}, varData []}}
      SectionFilter {allowedMaterialSlots []}
     }}
   )
   &InitialMeshOrientation(
    Mesh Orientation Mode=::NewEnumerator7
    Rotation=&RandomRangeVector(Minimum=<0 0 0>, Maximum=<1 1 1>)
    Use Rotation Vector=true
   )
   /SetVariables(Particles.Position=&AddVectorToPosition(Position=$Particles.Position, Vector=<0 0 60>)) {isSetParameters=true, module=SetVariables_#2}
   &AddVelocity(Module.Use Falloff Distance=false, Origin Offset=<0 0 140>, Override Seed=false, Velocity Mode=::NewEnumerator1, Velocity Speed=200)
   &ApplyInitialForces()
  ]
  particleUpdateScript [
   &ParticleState()
   &Drag(
    Drag=&FloatFromCurve(
     FloatCurve=DataInterface() {propertyValues {Curve {keys [
         {arriveTangent=-3.6239, interpMode=RCIM_Cubic, leaveTangent=-3.6239, tangentMode=RCTM_User, time=0, value=1}
         {interpMode=RCIM_Cubic, tangentMode=RCTM_User, time=0.3, value=0}
        ]}}}
     Scale Curve=8
    )
    Use Linear Drag=true
    Use Rotational Drag=true
   ) {module=Drag001}
   &AccelerationForce(Acceleration=&RandomRangeVector(Minimum=<-90 0 -90>, Maximum=<90 222 -90>))
   &SolveForcesAndVelocity(Clamp Velocity=false, Limit Acceleration=false)
   &DynamicMaterialParameters(
    Index 0 Param 1=1
    Index 0 Param 2=&FloatFromCurve(FloatCurve=DataInterface() {propertyValues {Curve {keys [
         {arriveTangent=-0.7948, interpMode=RCIM_Cubic, leaveTangent=-0.7949, tangentMode=RCTM_User, time=0.2519, value=0.1057}
         <at 0.7709 -10.0646 RCIM_Cubic>
        ]}}})
    Index 0 Param 4=&FloatFromCurve(FloatCurve=DataInterface() {propertyValues {Curve {keys [<at 0 6.7593 RCIM_Cubic>,<at 0.7 0 RCIM_Cubic>]}}})
    Module.Param0WriteEnabled=true
    Module.Param1WriteEnabled=true
    Module.Param2WriteEnabled=true
    Module.Param3WriteEnabled=true
   ) {module=DynamicMaterialParameters001}
   &ScaleColor(
    Scale Alpha=&FloatFromCurve(FloatCurve=DataInterface() {propertyValues {Curve {keys [
         <at -0.0046 0.8475 RCIM_Cubic>
         {arriveTangent=-1.3398, interpMode=RCIM_Cubic, leaveTangent=-1.3398, time=0.4249, value=0.5912}
         <at 1.0011 -0.5 RCIM_Cubic>
        ]}}})
    Scale RGB=&MakeVectorFromLinearColorRGB(Color=&ColorFromCurve(ColorCurve=DataInterface() {propertyValues {
        AlphaCurve {keys [<at 0 1 RCIM_Cubic>,<at 1 1 RCIM_Cubic>]}
        BlueCurve {keys [<at 0 1.269 RCIM_Cubic>,<at 0.0603 1.269>,<at 0.1297 0.4039>,<at 0.4378 0.1245>]}
        GreenCurve {keys [<at 0 0.0961 RCIM_Cubic>,<at 0.0603 0.0961>,<at 0.1297 0>,<at 0.4378 0.0424>]}
        RedCurve {keys [<at 0 3 RCIM_Cubic>,<at 0.0603 3>,<at 0.1297 1>,<at 0.4378 0.0764>]}
       }}))
    ScaleA=true
    ScaleRGB=true
   ) {module=ScaleColor001}
   &ScaleMeshSize(Scale Factor=&VectorFromFloat(Value=&FloatFromCurve(FloatCurve=DataInterface() {propertyValues {Curve {keys [<at 0.001 0.5>,<at 0.2948 1>,<at 1.0009 0.8072>]}}})))
   &CameraOffset(Camera Offset Amount=33)
  ]
  rendererClasses [&NiagaraMeshRendererProperties]
  renderers [{rendererClass=&NiagaraMeshRendererProperties, rendererIndex=0}]
 }
 Smoke_HeartShape_Outline {
  emitterUpdateScript [&EmitterState()]
  particleSpawnScript [
   &InitializeParticle(
    Color Mode=::NewEnumerator0
    Lifetime Max=2.5
    Lifetime Min=0.7
    Lifetime Mode=::NewEnumerator1
    Mesh Scale Mode=::NewEnumerator1
    Mesh Uniform Scale Max=0.65
    Mesh Uniform Scale Min=0.75
    Sprite Size Mode=::NewEnumerator3
    Uniform Sprite Size=15
    UsePositionOffset=false
    Write Material Random=true
   )
   &StaticMeshLocation(
    MaskSpawnGroup=false
    Mesh Sampling Type:ENiagara_StaticSamplingMode=::NewEnumerator2
    OffsetAlongNormal=false
    OverrideRandomSeed=false
    SamplePosOffset=false
    Static Mesh=DataInterface() {propertyValues {
      DefaultMesh=&SM_Emote_BoomShot_Heart
      InstanceIndex=-1
      LODIndex=0
      PreviewMesh=&SM_Emote_BoomShot_Heart
      SoftSourceActor=null
      SourceMode=DefaultMeshOnly
      bAllowSamplingFromStreamingLODs=false
      bCaptureTransformsPerFrame=true
      bUsePhysicsBodyVelocity=false
      FilteredSockets []
      LODIndexUserParameter {parameter {name=None, typeDefHandle {registeredTypeIndex=108}, varData []}}
      MeshParameterBinding {parameter {name=None, typeDefHandle {registeredTypeIndex=1}, varData []}}
      SectionFilter {allowedMaterialSlots []}
     }}
   )
   &InitialMeshOrientation(
    Mesh Orientation Mode=::NewEnumerator7
    Rotation=&RandomRangeVector(Minimum=<0 0 0>, Maximum=<1 1 1>)
    Use Rotation Vector=true
   )
   /SetVariables(Particles.Position=&AddVectorToPosition(Position=$Particles.Position, Vector=<0 0 60>)) {isSetParameters=true, module=SetVariables_#2}
   &ApplyInitialForces()
  ]
  particleUpdateScript [
   &ParticleState()
   &Drag(
    Drag=&FloatFromCurve(
     FloatCurve=DataInterface() {propertyValues {Curve {keys [
         {arriveTangent=-3.6239, interpMode=RCIM_Cubic, leaveTangent=-3.6239, tangentMode=RCTM_User, time=0, value=1}
         {interpMode=RCIM_Cubic, tangentMode=RCTM_User, time=0.3, value=0}
        ]}}}
     Scale Curve=8
    )
    Use Linear Drag=true
    Use Rotational Drag=true
   ) {module=Drag001}
   &AccelerationForce(Acceleration=&RandomRangeVector(Minimum=<-90 0 -90>, Maximum=<90 222 -90>))
   &SolveForcesAndVelocity(Clamp Velocity=false, Limit Acceleration=false)
   &DynamicMaterialParameters(
    Index 0 Param 1=1
    Index 0 Param 2=&FloatFromCurve(FloatCurve=DataInterface() {propertyValues {Curve {keys [
         {arriveTangent=-0.7948, interpMode=RCIM_Cubic, leaveTangent=-0.7949, tangentMode=RCTM_User, time=0.2519, value=0.1057}
         <at 0.7709 -10.0646 RCIM_Cubic>
        ]}}})
    Index 0 Param 4=&FloatFromCurve(FloatCurve=DataInterface() {propertyValues {Curve {keys [<at 0 6.7593 RCIM_Cubic>,<at 0.7 0 RCIM_Cubic>]}}})
    Module.Param0WriteEnabled=true
    Module.Param1WriteEnabled=true
    Module.Param2WriteEnabled=true
    Module.Param3WriteEnabled=true
   ) {module=DynamicMaterialParameters001}
   &ScaleColor(
    Scale Alpha=&FloatFromCurve(FloatCurve=DataInterface() {propertyValues {Curve {keys [
         <at -0.0046 0.8475 RCIM_Cubic>
         {arriveTangent=-1.3398, interpMode=RCIM_Cubic, leaveTangent=-1.3398, time=0.4249, value=0.5912}
         <at 1.0011 -0.5 RCIM_Cubic>
        ]}}})
    Scale RGB=&MakeVectorFromLinearColorRGB(Color=&ColorFromCurve(ColorCurve=DataInterface() {propertyValues {
        AlphaCurve {keys [<at 0 1 RCIM_Cubic>,<at 1 1 RCIM_Cubic>]}
        BlueCurve {keys [<at 0 1.269 RCIM_Cubic>,<at 0.0603 1.269>,<at 0.1297 0.4039>,<at 0.4378 0.1245>]}
        GreenCurve {keys [<at 0 0.0961 RCIM_Cubic>,<at 0.0603 0.0961>,<at 0.1297 0>,<at 0.4378 0.0424>]}
        RedCurve {keys [<at 0 3 RCIM_Cubic>,<at 0.0603 3>,<at 0.1297 1>,<at 0.4378 0.0764>]}
       }}))
    ScaleA=true
    ScaleRGB=true
   ) {module=ScaleColor001}
   &CameraOffset(Camera Offset Amount=33)
  ]
  rendererClasses [&NiagaraMeshRendererProperties]
  renderers [{rendererClass=&NiagaraMeshRendererProperties, rendererIndex=0}]
 }
 Sparks {
  emitterUpdateScript [&EmitterState(
    Enable Distance Culling=true
    Max Distance Response=::KillImmediately
    Module.bApplySpawnCountScaleByDistance=true
    Module.bUseMaxDistance=true
    Module.bUseMinDistance=false
    Scalability Mode=::NewEnumerator1
    Scale Spawn Count=true
    Spawn Count Scale By Distance=DataInterface() {propertyValues {Curve {keys [<at 0.3676 0.9657>,<at 0.9962 0.4982>]}}}
   )]
  particleSpawnScript [
   &InitializeParticle(
    Color=<rgba 3 3 3 1>
    Lifetime Max=0.45
    Lifetime Min=0.166
    Lifetime Mode=::NewEnumerator1
    Sprite Size Mode=::NewEnumerator3
    Uniform Sprite Size=22
    UsePositionOffset=false
    Write Material Random=true
   )
   &ShapeLocation(
    MaskSpawnGroup=false
    Override Seed=false
    Sphere Radius=&Multiply_Float(A=9, B=&FN_FrontEndFloatModifier(Gameplay=1, Front End=$User.FrontEndScale))
    Sphere Surface Distribution=1
    UseOwnerScale=true
   )
   &AddVelocity(
    Cone Angle=360
    Module.UseConeFalloff=false
    Module.UseVelDistribution=false
    Override Seed=false
    Velocity Mode=::NewEnumerator2
    Velocity Speed=500
   ) {module=AddVelocity002}
   &DynamicMaterialParameters(
    Index 0 Param 1=3
    Index 0 Param 2=20
    Index 0 Param 3=1
    Module.Param0WriteEnabled=true
    Module.Param1WriteEnabled=true
    Module.Param2WriteEnabled=true
    Module.Param3WriteEnabled=true
   )
  ]
  particleUpdateScript [
   &ParticleState()
   &GravityForce(Gravity=<0 0 -666>)
   &Drag(Drag=&UniformRangedFloat(Minimum=2, Maximum=6), Use Linear Drag=true, Use Rotational Drag=false)
   &SolveForcesAndVelocity(Clamp Velocity=false, Limit Acceleration=false)
   &ScaleSpriteSize(Scale Factor=&Vector2DFromCurve(
     Vector2Curve=DataInterface() {propertyValues {
       XCurve {keys [
         {arriveTangent=5.9728, interpMode=RCIM_Cubic, leaveTangent=5.9728, tangentMode=RCTM_User, time=0, value=0.1}
         {arriveTangent=-0.1, interpMode=RCIM_Cubic, leaveTangent=-0.1, tangentMode=RCTM_User, time=0.18, value=0.8}
         {interpMode=RCIM_Cubic, tangentMode=RCTM_User, time=1, value=0}
        ]}
       YCurve {keys [
         {arriveTangent=6.2715, interpMode=RCIM_Cubic, leaveTangent=6.2715, tangentMode=RCTM_User, time=0, value=0.1}
         {arriveTangent=-0.1001, interpMode=RCIM_Cubic, leaveTangent=-0.1001, tangentMode=RCTM_User, time=0.18, value=0.8}
         {interpMode=RCIM_Cubic, tangentMode=RCTM_User, time=0.9988, value=0}
        ]}
      }}
     ScaleCurve=<1 1>
    ))
   &ScaleSpriteSizeBySpeed(Max Scale Factor=<0.5 1>, Min Scale Factor=<0.5 1>, Velocity Threshold=666)
   &Color(Color=&LerpLinearColors(
     StartColor=&ColorFromCurve(ColorCurve=DataInterface() {propertyValues {
        AlphaCurve {keys [<at 0.333 1>,<at 1 0>]}
        BlueCurve {keys [<at 0 0.24>,<at 0.166 0.2>,<at 0.333 0>,<at 0.75 0>]}
        GreenCurve {keys [<at 0 3.12>,<at 0.166 2.6>,<at 0.333 0.6667>,<at 0.75 0.15>]}
        RedCurve {keys [<at 0 6>,<at 0.166 5>,<at 0.333 4>,<at 0.75 3>]}
       }})
     EndColor=&ColorFromCurve(ColorCurve=DataInterface() {propertyValues {
        AlphaCurve {keys [<at 0.333 1>,<at 1 0>]}
        BlueCurve {keys [<at 0 6>,<at 0.166 5>,<at 0.333 0.2031>,<at 0.75 0.0885>]}
        GreenCurve {keys [<at 0 6>,<at 0.166 5>,<at 0.333 0.2031>,<at 0.75 0.0885>]}
        RedCurve {keys [<at 0 6>,<at 0.166 5>,<at 0.333 0.2031>,<at 0.75 0.0885>]}
       }})
     LerpFactor=$User.VariantPicker
    ))
  ]
  rendererClasses [&NiagaraSpriteRendererProperties]
  renderers [{rendererClass=&NiagaraSpriteRendererProperties}]
 }
 Text_KA_BOOM {
  emitterUpdateScript [
   &EmitterState()
   &SpawnBurst_Instantaneous(Spawn Time=0.72, Use Spawn Probability=false, UseLoopCountLimit=false)
   &SpawnBurst_Instantaneous(Spawn Group=1, Spawn Time=0.82, Use Spawn Probability=false, UseLoopCountLimit=false) {module=SpawnBurst_Instantaneous001}
  ]
  particleSpawnScript [
   /SetVariables(Particles.VisibilityTag=$Emitter.SpawnGroup) {isSetParameters=true}
   &InitializeParticle(Lifetime=1.5, Mesh Scale Mode=::NewEnumerator3, Mesh Uniform Scale=0.25, UsePositionOffset=false, Write Material Random=true) {module=InitializeParticle001}
   /SetVariables(Particles.BoomLifetime=1.3) {isSetParameters=true}
   /SetVariables(Particles.Position=&AddVectorToPosition(Position=$Particles.Position, Vector=<0 0 75>)) {isSetParameters=true}
   &AddVelocity(
    Velocity=<0 0 1>
    Velocity Speed Scale=&Lerp_Float(A=20, B=0, Alpha=&MakeFloatFromInt(Int=$Particles.VisibilityTag))
   ) {module=AddVelocity001}
   &ApplyInitialForces()
  ]
  particleUpdateScript [
   &ParticleState()
   &Drag(
    Drag=&FloatFromCurve(
     FloatCurve=DataInterface() {propertyValues {Curve {keys [
         {interpMode=RCIM_Cubic, tangentMode=RCTM_User, time=0, value=0}
         {arriveTangent=9.269, interpMode=RCIM_Cubic, leaveTangent=-9.7228, tangentMode=RCTM_Break, time=0.4, value=1}
         {arriveTangent=-0.1416, interpMode=RCIM_Cubic, leaveTangent=-0.1416, tangentMode=RCTM_User, time=0.6, value=0}
         <at 1 0 RCIM_Cubic>
        ]}}}
     Scale Curve=&Lerp_Float(A=4, B=5.5, Alpha=&MakeFloatFromInt(Int=$Particles.VisibilityTag))
    )
    Use Linear Drag=true
    Use Rotational Drag=true
   ) {module=Drag001}
   &SolveForcesAndVelocity(Clamp Velocity=false, Limit Acceleration=false)
   &ScaleMeshSize(Scale Factor=&Multiply_Vector(
     A=/Niagara/DynamicInputs/ScaleByCurve/ScaleVectorByCurve(
      InitialValue:Vector3f=<1 1 1>
      ScaleCurve:NiagaraDataInterfaceVectorCurve=DataInterface() {propertyValues {
        XCurve {keys [
          <at 0 0 RCIM_Cubic>
          {arriveTangent=2.2843, interpMode=RCIM_Cubic, leaveTangent=2.2843, time=0.3178, value=1.2}
          {arriveTangent=-0.4851, interpMode=RCIM_Cubic, leaveTangent=-0.4851, time=0.4378, value=1}
          {arriveTangent=-1.7786, interpMode=RCIM_Cubic, leaveTangent=-1.7786, time=0.75, value=0.9903}
          <at 1 0 RCIM_Cubic>
         ]}
        YCurve {keys [
          <at 0 0 RCIM_Cubic>
          {arriveTangent=3.2407, interpMode=RCIM_Cubic, leaveTangent=3.2407, time=0.25, value=1.3991}
          {arriveTangent=-0.6628, interpMode=RCIM_Cubic, leaveTangent=-0.6628, time=0.37, value=1.1991}
          {arriveTangent=-1.9033, interpMode=RCIM_Cubic, leaveTangent=-1.9033, time=0.75, value=1.0676}
          <at 1 0 RCIM_Cubic>
         ]}
        ZCurve {keys [
          <at 0 0 RCIM_Cubic>
          {arriveTangent=3.2407, interpMode=RCIM_Cubic, leaveTangent=3.2407, time=0.25, value=1.3991}
          {arriveTangent=-0.6628, interpMode=RCIM_Cubic, leaveTangent=-0.6628, time=0.37, value=1.1991}
          {arriveTangent=-1.9033, interpMode=RCIM_Cubic, leaveTangent=-1.9033, time=0.75, value=1.0676}
          <at 1 0 RCIM_Cubic>
         ]}
       }}
      CurveIndex:NiagaraFloat=$Particles.NormalizedAge
     )
     B=&VectorFromFloat(Value=&FloatFromCurve(FloatCurve=DataInterface() {propertyValues {Curve {keys [<at 0.75 1 RCIM_Cubic>,<at 1 0 RCIM_Cubic>]}}}))
    ))
   &ScaleColor(Scale RGB=&MakeVectorFromLinearColorRGB(Color=<rgba 1 1 1 1>), ScaleA=true, ScaleRGB=true)
   /SetVariables(Particles.BoomUpdateAge:NiagaraFloat=&Divide_Float(A=$Particles.Age, B=$Particles.Initial.BoomLifetime)) {isSetParameters=true}
   /SetVariables(
    Particles.BoomDMP:Vector4f=&MakeVector4(
     X=&FloatFromCurve(
      FloatCurve=DataInterface() {propertyValues {Curve {keys [{tangentMode=RCTM_User, time=0, value=1}, <at 1 4>]}}}
      CurveIndex=$Particles.BoomUpdateAge
     )
     Y=1
     Z=&FloatFromCurve(
      FloatCurve=DataInterface() {propertyValues {Curve {keys [
          {arriveTangent=-17.7515, interpMode=RCIM_Cubic, leaveTangent=-17.7515, tangentMode=RCTM_User, time=-0.0046, value=1.0113}
          {arriveTangent=-0.2845, interpMode=RCIM_Cubic, leaveTangent=-0.2845, tangentMode=RCTM_User, time=0.1, value=0.065}
          {arriveTangent=0.4619, interpMode=RCIM_Cubic, leaveTangent=0.4619, tangentMode=RCTM_User, time=0.6513, value=0.0607}
          {arriveTangent=3.5589, interpMode=RCIM_Cubic, leaveTangent=3.5589, tangentMode=RCTM_User, time=1, value=1}
         ]}}}
      CurveIndex=$Particles.BoomUpdateAge
     )
     W=150
    )
    Particles.BoomScale:Vector3f=&Multiply_Vector(
     A=&ScaleVectorByCurve(
      InitialValue:Vector3f=<1 1 1>
      ScaleCurve:NiagaraDataInterfaceVectorCurve=DataInterface() {propertyValues {
        XCurve {keys [
          <at 0 0 RCIM_Cubic>
          {arriveTangent=2.2843, interpMode=RCIM_Cubic, leaveTangent=2.2843, time=0.3178, value=1.2}
          {arriveTangent=-0.4627, interpMode=RCIM_Cubic, leaveTangent=-0.4627, time=0.4378, value=1}
          {arriveTangent=-1.7786, interpMode=RCIM_Cubic, leaveTangent=-1.7786, time=0.75, value=1}
          <at 1 0 RCIM_Cubic>
         ]}
        YCurve {keys [
          <at 0 0 RCIM_Cubic>
          {arriveTangent=2.2886, interpMode=RCIM_Cubic, leaveTangent=2.2886, time=0.3167, value=1.4998}
          {arriveTangent=-1.1535, interpMode=RCIM_Cubic, leaveTangent=-1.1535, time=0.5159, value=1.1808}
          {arriveTangent=-2.4394, interpMode=RCIM_Cubic, leaveTangent=-2.4394, time=0.75, value=1}
          <at 1 0 RCIM_Cubic>
         ]}
        ZCurve {keys [
          <at 0 0 RCIM_Cubic>
          {arriveTangent=3.0427, interpMode=RCIM_Cubic, leaveTangent=3.0427, time=0.25, value=1.3258}
          {arriveTangent=-0.4, interpMode=RCIM_Cubic, leaveTangent=-0.4, time=0.37, value=1.1258}
          {arriveTangent=-1.787, interpMode=RCIM_Cubic, leaveTangent=-1.787, time=0.75, value=1.1258}
          <at 1 0 RCIM_Cubic>
         ]}
       }}
      CurveIndex:NiagaraFloat=$Particles.BoomUpdateAge
     )
     B=&Multiply_VectorByFloat(Vector=<1 1 1>, Float=0.25)
    )
   ) {isSetParameters=true}
   &DynamicMaterialParameters(
    Index 0 Param 1=&FloatFromCurve(FloatCurve=DataInterface() {propertyValues {Curve {keys [
         {arriveTangent=5.6948, interpMode=RCIM_Cubic, leaveTangent=5.6948, tangentMode=RCTM_User, time=0, value=0}
         {arriveTangent=-0.5168, interpMode=RCIM_Cubic, leaveTangent=-0.5168, tangentMode=RCTM_User, time=0.8442, value=1.658}
         {interpMode=RCIM_Cubic, leaveTangent=0.574, time=0.98, value=1.3}
        ]}}})
    Index 0 Param 3=&FloatFromCurve(FloatCurve=DataInterface() {propertyValues {Curve {keys [
         {arriveTangent=-6.3759, interpMode=RCIM_Cubic, leaveTangent=-6.3759, tangentMode=RCTM_User, time=-0.0046, value=1.0113}
         {arriveTangent=-0.0113, interpMode=RCIM_Cubic, leaveTangent=-0.0113, time=0.4003, value=0.002}
         {arriveTangent=0.0913, interpMode=RCIM_Cubic, leaveTangent=0.0913, tangentMode=RCTM_User, time=1, value=1}
        ]}}})
    Index 0 Param 4=80
    Module.Param0WriteEnabled=true
    Module.Param1WriteEnabled=true
    Module.Param2WriteEnabled=true
    Module.Param3WriteEnabled=true
   )
   &CameraOffset(Camera Offset Amount=20)
  ]
  rendererClasses [&NiagaraMeshRendererProperties]
  renderers [
   {rendererClass=&NiagaraMeshRendererProperties, rendererIndex=0}
   {rendererClass=&NiagaraMeshRendererProperties, rendererIndex=1}
  ]
 }
}
summary {
 systemName=NS_Emote_BoomShot
 emitters [
  {
   emitterName=Bombs_HeartShape
   rendererClasses [&NiagaraMeshRendererProperties, &NiagaraSpriteRendererProperties]
  }
  {emitterName=ImpactFlash, rendererClasses [&NiagaraSpriteRendererProperties]}
  {emitterName=Sparks, rendererClasses [&NiagaraSpriteRendererProperties]}
  {emitterName=Hearts_Small, rendererClasses [&NiagaraSpriteRendererProperties]}
  {emitterName=Smoke_HeartShape_Outline, rendererClasses [&NiagaraMeshRendererProperties]}
  {emitterName=Smoke_HeartShape, rendererClasses [&NiagaraMeshRendererProperties]}
  {
   emitterName=Glow_BenDayDots
   rendererClasses [&NiagaraSpriteRendererProperties, &NiagaraMeshRendererProperties]
  }
  {emitterName=Flash_BenDayDots, rendererClasses [&NiagaraSpriteRendererProperties]}
  {emitterName=Text_KA_BOOM, rendererClasses [&NiagaraMeshRendererProperties]}
 ]
 userVariables [
  User.bIsFrontEnd {type=NiagaraBool}
  User.bIsFrontEndPreview {type=NiagaraBool}
  User.Color Array {
   type=NiagaraDataInterfaceArrayColor
   description=""
   defaultValue {
    struct=NiagaraExt_VariableValue_DataInterface
    value {
     dataInterface=&self:NiagaraDataInterfaceArrayColor_0
     dataInterfaceClass=&NiagaraDataInterfaceArrayColor
    }
   }
  }
  User.FrontEndScale {type=NiagaraFloat, defaultValue=0.6, description=""}
  User.VariantPicker {type=NiagaraFloat, defaultValue=0, description=""}
 ]
}
